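# Translucent, lit sprite material for camera-facing ladybug particles in environment effects, usable by both Niagara and Cascade sprites.
# A SubUV ladybug texture supplies color through ConvertFromDiffSpec, and its alpha is multiplied by particle alpha.
# A time-of-day scalar, rounded to a hard 0 or 1 by add, floor and clamp, multiplies that alpha so the ladybugs show from morning to evening and vanish at night.
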Material /Game/Effects/Fort_Effects/World/Enviroment/Ladybugs/M_Ladybug_CameraFace
graph {
 properties {
  bUsedWithNiagaraSprites=true
  bUsedWithParticleSprites=true
  blendMode=BLEND_Translucent
  pixelDepthOffsetMode=PDOM_Legacy
  translucencyLightingMode=TLM_VolumetricPerVertexNonDirectional
  usedShadingModels=Default Lit
 }
 nodes [
  n1 = ParticleSubUV!(UVs, Tex, Apply View MipBias) {texture=&/T_LadyBug_SubUV}
  MaterialFunctionCall(DiffuseColor=@n1.RGB) {materialFunction=&ConvertFromDiffSpec}
  n2 = Constant() {r=1}
  n3 = Add(A=MaterialFunctionCall(Value_Morning=@n2, Value_Day=@n2, Value_Evening=@n2, Value_Night=Constant()) {materialFunction=&Scalar_Time_of_Day}.Result) {constB=0.5}
  Multiply(A=Clamp(None=Floor(None=@n3)), B=Multiply(A=@n1.A, B=ParticleColor().A))
 ]
}
